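# Physics asset for the Butter Jack backpack cosmetic, defining how its slushy bottle and straw can move under simulation.
# Capsule bodies cover the bottle base, the slushy section and seven straw bones.
# The slushy hangs off the bottle base with moderate swing, and the straw continues from it as a serial chain.
# Each straw joint allows only slight, asymmetric swing and almost no twist, so the chain bends gradually along its length.
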PhysicsAsset /BRCosmetics/Accessories/FORT_Backpacks/Backpack_M_MED_Butter_Jack/Meshes/M_MED_Butter_Jack_Pack_Physics
bodies {
 bottle_base [{shapeType=Capsule, length=10, radius=10, shapeName=bottle_base_capsule, center=<-8 0.5 0>, rotation=<rot 88 0 0>}]
 dyn_bottle_slushy [{shapeType=Capsule, length=1, radius=7, shapeName=dyn_bottle_slushy_capsule, center=<-2.5 0 0>, rotation=<rot 90 0 0>}]
 dyn_straw_01 [{shapeType=Capsule, length=8, radius=1, shapeName=dyn_straw_01_capsule, center=<4 0 0>, rotation=<rot 90 0 0>}]
 dyn_straw_02 [{shapeType=Capsule, length=8, radius=1, shapeName=dyn_straw_02_capsule, center=<4.5 0 0>, rotation=<rot 90 0 0>}]
 dyn_straw_03 [{shapeType=Capsule, length=14, radius=1, shapeName=dyn_straw_03_capsule, center=<7 0 0>, rotation=<rot 90 0 0>}]
 dyn_straw_04 [{shapeType=Capsule, length=8, radius=1, shapeName=dyn_straw_04_capsule, center=<4.5 0 0>, rotation=<rot 90 0 0>}]
 dyn_straw_05 [{shapeType=Capsule, length=6, radius=1, shapeName=dyn_straw_05_capsule, center=<3.5 0 0>, rotation=<rot 90 0 0>}]
 dyn_straw_06 [{shapeType=Capsule, length=10, radius=1, shapeName=dyn_straw_06_capsule, center=<6 0 0>, rotation=<rot 90 0 0>}]
 dyn_straw_07 [{shapeType=Capsule, length=9, radius=1, shapeName=dyn_straw_07_capsule, center=<5 0 0>, rotation=<rot 90 0 0>}]
}
constraints [
 {bone1Name=dyn_bottle_slushy, bone2Name=bottle_base, swing1LimitDegrees=15, swing2LimitDegrees=15, twistLimitDegrees=1}
 {bone1Name=dyn_straw_01, bone2Name=dyn_bottle_slushy, swing1LimitDegrees=8, swing2LimitDegrees=3, twistLimitDegrees=1}
 {bone1Name=dyn_straw_02, bone2Name=dyn_straw_01, swing1LimitDegrees=8, swing2LimitDegrees=3, twistLimitDegrees=1}
 {bone1Name=dyn_straw_03, bone2Name=dyn_straw_02, swing1LimitDegrees=8, swing2LimitDegrees=3, twistLimitDegrees=1}
 {bone1Name=dyn_straw_04, bone2Name=dyn_straw_03, swing1LimitDegrees=8, swing2LimitDegrees=3, twistLimitDegrees=1}
 {bone1Name=dyn_straw_05, bone2Name=dyn_straw_04, swing1LimitDegrees=8, swing2LimitDegrees=3, twistLimitDegrees=1}
 {bone1Name=dyn_straw_06, bone2Name=dyn_straw_05, swing1LimitDegrees=8, swing2LimitDegrees=3, twistLimitDegrees=1}
 {bone1Name=dyn_straw_07, bone2Name=dyn_straw_06, swing1LimitDegrees=8, swing2LimitDegrees=3, twistLimitDegrees=1}
]
