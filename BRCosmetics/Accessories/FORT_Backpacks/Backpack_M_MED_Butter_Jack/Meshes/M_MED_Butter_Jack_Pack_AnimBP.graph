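# Animation Blueprint for the Butter Jack backpack cosmetic that adds physics motion and adjusts the straw bones during skydiving.
# A RigidBody node simulates the pack's physics asset from the reference pose, relative to bottle_base, with scaled, clamped component motion and overridden gravity.
# Two chained Transform Bone nodes then add rotations to dyn_straw_01 and dyn_straw_02, enabled by bIsSkyDiving.
# The event graph has no connected logic.
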AnimBlueprint /BRCosmetics/Accessories/FORT_Backpacks/Backpack_M_MED_Butter_Jack/Meshes/M_MED_Butter_Jack_Pack_AnimBP
target_skeleton=&/M_MED_Butter_Jack_Pack_Skeleton
graphs [
 AnimGraph {nodes [
   n1 = RigidBody(ComponentPose=Local To Component(LocalPose=Local Space Ref Pose())) {
    bClampLinearTranslationLimitToRefPose=true
    bOverrideWorldGravity=true
    overridePhysicsAsset=&/M_MED_Butter_Jack_Pack_Physics
    simulationSpace=BaseBoneSpace
    alphaBoolBlend {blendOption=Linear}
    baseBoneRef {boneName=bottle_base}
    componentAppliedLinearAccClamp=<3000 3000 3000>
    componentLinearAccScale=<0.3 0.3 0.45>
    componentLinearVelScale=<0.15 0.15 0.225>
    overrideWorldGravity {z=-980}
   }
   n2 = Get bIsSkyDiving()
   n3 = "Transform (Modify) Bone - Bone: dyn_straw_01"!(ComponentPose=@n1, bAlphaBoolEnabled:bool=@n2, Translation="0,0,0", Rotation="20,-25,0.000", Scale="1,1,1") -> (Pose) {
    alphaInputType=Bool
    rotationMode=BMM_Additive
    rotationSpace=BCS_ParentBoneSpace
    translationMode=BMM_Additive
    translationSpace=BCS_ParentBoneSpace
    alphaBoolBlend {blendOption=Linear}
    boneToModify {boneName=dyn_straw_01}
    rotation {pitch=20, yaw=-25}
   }
   n4 = "Transform (Modify) Bone - Bone: dyn_straw_02"!(ComponentPose=@n3, bAlphaBoolEnabled:bool=@n2, Translation="0,0,0", Rotation="0,-20,-30", Scale="1,1,1") -> (Pose) {
    alphaInputType=Bool
    rotationMode=BMM_Additive
    rotationSpace=BCS_ParentBoneSpace
    translationMode=BMM_Additive
    translationSpace=BCS_ParentBoneSpace
    alphaBoolBlend {blendOption=Linear}
    boneToModify {boneName=dyn_straw_02}
    rotation {roll=-30, yaw=-20}
   }
   Output Pose(Result=Component To Local(ComponentPose=@n4)) {layerGroup=None}
  ]}
 EventGraph {nodes [
   TryGetPawnOwner()
   Event BlueprintUpdateAnimation()
  ]}
]
variables []
